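# Material function for the Athena landscape gravel layer, outputting full material attributes plus separate Height and untinted base color outputs.
# One gravel texture, sampled with optional input UVs, has its RGB multiplied by the GravelTint color from the landscape parameter collection.
# Its red channel blends between Gravel_SpecMin and Gravel_SpecMax for specular, roughness is constant, and alpha feeds both opacity and height.
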MaterialFunction /FigmentEnvironment/FN_6_20/Athena/Environments/Landscape/MaterialFunctions/Standard/MF_Athena_Gravel_01
graph {nodes [
  Constant3Vector() {constant=<rgba 0 0 1 0>}
  n1 = TextureSampleParameter2D(UVs=FunctionInput(Preview=TextureCoordinate()) {bUsePreviewValueAsDefault=true, inputType=FunctionInput_Vector2}) {
   group=Gravel
   parameterName=Gravel
   samplerSource=SSM_Wrap_WorldGroupSettings
   texture=/FigmentEnvironment/FN_6_20/Packages/DS_Fortnight/WhiteRockCustom_D
  }
  n2 = ComponentMask(None=CollectionParameter() {collection=/FigmentEnvironment/FN_6_20/Athena/Environments/Landscape/MPC/MPC_Landscape, parameterName=GravelTint}) {b=true, g=true, r=true}
  n3 = ScalarParameter() {defaultValue=-0.45, group=Gravel, parameterName=Gravel_SpecMin}
  n4 = ScalarParameter() {defaultValue=2, group=Gravel, parameterName=Gravel_SpecMax}
  FunctionOutput(None=MakeMaterialAttributes!(
    BaseColor=Multiply(A=@n1.RGB, B=@n2)
    Metallic
    Specular=LinearInterpolate(A=@n3, B=@n4, Alpha=@n1.R)
    Roughness=Constant() {r=1}
    Anisotropy
    EmissiveColor
    Opacity=@n1.A
    OpacityMask
    Normal
    Tangent
    WorldPositionOffset
    SubsurfaceColor
    ClearCoat
    ClearCoatRoughness
    AmbientOcclusion
    Refraction
    CustomizedUVs_0
    CustomizedUVs_1
    CustomizedUVs_2
    CustomizedUVs_3
    CustomizedUVs_4
    CustomizedUVs_5
    CustomizedUVs_6
    CustomizedUVs_7
    PixelDepthOffset
    ShadingModel
    Displacement
   ))
  FunctionOutput(None=@n1.A) {outputName=Height, sortPriority=1}
  FunctionOutput(None=@n1.RGB) {outputName=Plain_BaseColor, sortPriority=2}
 ]}
