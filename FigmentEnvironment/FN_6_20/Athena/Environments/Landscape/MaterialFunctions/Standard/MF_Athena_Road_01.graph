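# Material function for the Athena landscape that builds street road material attributes from diffuse, mask and normal textures sharing one UV input.
# Base color is the diffuse scaled by a brightness parameter, and its alpha drives opacity.
# A UseSimple static switch, on by default, derives specular from diffuse red with constant roughness and a flat normal.
# When off, it uses mask red for specular, mask blue for roughness range and the normal map.
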MaterialFunction /FigmentEnvironment/FN_6_20/Athena/Environments/Landscape/MaterialFunctions/Standard/MF_Athena_Road_01
graph {nodes [
  n1 = FunctionInput(Preview=TextureCoordinate()) {bUsePreviewValueAsDefault=true, inputType=FunctionInput_Vector2}
  n2 = TextureSample(UVs=@n1) {samplerSource=SSM_Wrap_WorldGroupSettings, texture=/FigmentEnvironment/FN_6_20/Packages/DS_Fortnight/SM/Textures/T_Street_Road_D}
  n3 = ScalarParameter() {defaultValue=1.1, group=Road, parameterName=Road_Brightness}
  n4 = LinearInterpolate(
   A=ScalarParameter() {defaultValue=-0.35, group=Road, parameterName=Road_Spec_2}
   B=ScalarParameter() {defaultValue=4, group=Road, parameterName=Road_Spec_3}
   Alpha=@n2.R
  )
  n5 = TextureSample(UVs=@n1) {samplerType=SAMPLERTYPE_Masks, texture=/FigmentEnvironment/FN_6_20/Packages/DS_Fortnight/SM/Textures/T_Street_Road_M}
  n6 = FunctionInput(Preview=StaticBool() {value=true}) {bUsePreviewValueAsDefault=true, inputName=UseSimple, inputType=FunctionInput_StaticBool}
  n7 = Constant() {r=1}
  n8 = LinearInterpolate(A=ScalarParameter() {parameterName=Road_Rough_Min}, B=ScalarParameter() {defaultValue=0.65, parameterName=Road_Rough_Max}, Alpha=@n5.B)
  n9 = Constant3Vector() {constant=<rgba 0 0 1 0>}
  n10 = TextureSample(UVs=@n1) {samplerType=SAMPLERTYPE_Normal, texture=/FigmentEnvironment/FN_6_20/Packages/DS_Fortnight/SM/Textures/T_Street_Road_N}
  FunctionOutput(None=MakeMaterialAttributes!(
    BaseColor=Multiply(A=@n2.RGB, B=@n3)
    Metallic
    Specular=StaticSwitch(True=@n4, False=@n5.R, Value=@n6)
    Roughness=StaticSwitch(True=@n7, False=@n8, Value=@n6)
    Anisotropy
    EmissiveColor
    Opacity=@n2.A
    OpacityMask
    Normal=StaticSwitch(True=@n9, False=@n10.RGB, Value=@n6)
    Tangent
    WorldPositionOffset
    SubsurfaceColor
    ClearCoat
    ClearCoatRoughness
    AmbientOcclusion
    Refraction
    CustomizedUVs_0
    CustomizedUVs_1
    CustomizedUVs_2
    CustomizedUVs_3
    CustomizedUVs_4
    CustomizedUVs_5
    CustomizedUVs_6
    CustomizedUVs_7
    PixelDepthOffset
    ShadingModel
    Displacement
   ))
 ]}
