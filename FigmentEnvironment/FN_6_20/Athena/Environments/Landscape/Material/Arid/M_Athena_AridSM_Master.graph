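# Master material for Athena's arid-biome static and spline meshes, matching cliffs, roads and river splines to landscape layers.
# Vertex colour channels drive height-lerp blends of sand or grass with mud, road, gravel and sediment-tinted cliff rock, with static switches selecting variants.
# An optional topo pass adjusts terrain colour, and normals are re-projected from world space.
# Desert roads add spline height offset; mobile and low quality get flat-normal, simplified attributes.
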Material /FigmentEnvironment/FN_6_20/Athena/Environments/Landscape/Material/Arid/M_Athena_AridSM_Master
graph {
 properties {
  bUseMaterialAttributes=true
  bUsedWithSplineMeshes=true
  pixelDepthOffsetMode=PDOM_Legacy
  refractionMethod=RM_IndexOfRefraction
  usedShadingModels=Default Lit
 }
 nodes [
  Constant() {r=1}
  LandscapeLayerCoords() {mappingScale=1525}
  n1 = MaterialFunctionCall!(
   In=MaterialFunctionCall(
    In=MaterialFunctionCall!(ProjContrast) {materialFunction=/FigmentEnvironment/FN_6_20/Athena/Environments/Landscape/MaterialFunctions/Arid/MF_Athena_Arid_Rock}.Mat
    Color=MaterialFunctionCall() {materialFunction=/FigmentEnvironment/FN_6_20/Athena/Environments/Landscape/MaterialFunctions/MF_Athena_SedimentGradient}.Result
   ) {materialFunction=/FigmentEnvironment/FN_6_20/Athena/Environments/Landscape/MaterialFunctions/MF_Athena_Material_BaseMultiply}.Result
   Color=ScalarParameter() {defaultValue=1.25, parameterName=Arid_CliffMesh_Brightness}
  ) {materialFunction=&MF_Athena_Material_BaseMultiply}
  n2 = MaterialFunctionCall(
   In=MaterialFunctionCall() {materialFunction=/FigmentEnvironment/FN_6_20/Athena/Environments/Landscape/MaterialFunctions/Standard/MF_Athena_Grass_01}.Result
   Color=CollectionParameter() {collection=/FigmentEnvironment/FN_6_20/Athena/Environments/Landscape/MPC/MPC_Landscape, parameterName=Arid_Grass_Tint}
  ) {materialFunction=&MF_Athena_Material_BaseMultiply}
  n3 = CollectionParameter() {collection=&MPC_Landscape, parameterName=Arid_Sand_TexSize}
  n4 = Divide(A=WorldPosition() {worldPositionShaderOffset=WPT_ExcludeAllShaderOffsets}.XYZ, B=Multiply(A=Abs(None=@n3)) {constB=-1})
  n5 = MaterialFunctionCall(
   In=MaterialFunctionCall!(UseRoughness=StaticBool(), UVs=ComponentMask(None=@n4) {g=true, r=true}) {materialFunction=/FigmentEnvironment/FN_6_20/Athena/Environments/Landscape/MaterialFunctions/Arid/MF_Athena_Arid_Sand}.Simple_Material
   Color=MaterialFunctionCall() {materialFunction=/FigmentEnvironment/FN_6_20/Athena/Environments/Landscape/MaterialFunctions/MF_Athena_ZSandColor}.Result
  ) {materialFunction=&MF_Athena_Material_BaseMultiply}
  n6 = StaticSwitchParameter(True=@n2.Result, False=@n5.Result) {parameterName=UseGrassAsBase}
  n7 = MaterialFunctionCall() {materialFunction=/FigmentEnvironment/FN_6_20/Athena/Environments/Landscape/MaterialFunctions/Standard/MF_Athena_Mud_01}
  n8 = CollectionParameter() {collection=&MPC_Landscape, parameterName=Arid_Mud_Tint}
  n9 = MaterialFunctionCall!(
   Material=@n7.Result
   NewBaseColor=Multiply(A=@n7.BaseColor_Plain, B=ComponentMask(None=@n8) {b=true, g=true, r=true})
   Mask=MaterialFunctionCall() {materialFunction=/FigmentEnvironment/FN_6_20/Athena/Environments/Landscape/MaterialFunctions/MF_Athena_BiomeMask}.Swamp
  ) {materialFunction=/Engine/Functions/MaterialLayerFunctions/MatLayerBlend_OverrideBaseColor}
  n10 = VertexColor()
  n11 = StaticSwitchParameter(True=OneMinus(None=@n10.A), False=@n10.A) {parameterName=IsDesertRoad}
  n12 = MaterialFunctionCall!(Mat attributes=@n2.Result) {materialFunction=/Engine/Functions/Engine_MaterialFunctions02/MatLayerBlend_BreakOpacity}
  n13 = StaticSwitchParameter(
   True=OneMinus(None=@n12.Result)
   False=MaterialFunctionCall(Mat attributes=@n7.Result) {materialFunction=&MatLayerBlend_BreakOpacity}.Result
  ) {parameterName=UseGrassAsBase}
  n14 = MaterialFunctionCall!(
   Base Material=@n6
   Top Material=@n9.Blended Material
   Alpha=MaterialFunctionCall(Transition Phase=Clamp(None=@n11), Height Texture=@n13, Contrast=ScalarParameter() {defaultValue=1, parameterName=Mud_Blend_Contrast}) {desc=Mud, materialFunction=&HeightLerp}.Alpha
  ) {materialFunction=/Engine/Functions/MaterialLayerFunctions/MatLayerBlend_Standard}
  n15 = VertexColor()
  n16 = Abs(None=ScalarParameter() {defaultValue=1024, group=MasterControl, parameterName=Texturesize_Z})
  n17 = ComponentMask(None=Divide(A=WorldPosition() {worldPositionShaderOffset=WPT_ExcludeAllShaderOffsets}.XYZ, B=Multiply(A=@n16) {constB=-1})) {g=true, r=true}
  n18 = MaterialFunctionCall!(UseSimple=StaticBool(), In=@n17) {materialFunction=/FigmentEnvironment/FN_6_20/Athena/Environments/Landscape/MaterialFunctions/Standard/MF_Athena_Road_01}
  n19 = MaterialFunctionCall(
   Transition Phase=Power(Base=@n15.R, Exp=ScalarParameter() {defaultValue=2, parameterName=RoadMud_Trans})
   Height Texture=MaterialFunctionCall(Mat attributes=@n18.Result) {materialFunction=&MatLayerBlend_BreakOpacity}.Result
   Contrast=ScalarParameter() {defaultValue=2, parameterName=RoadMud_Contrast}
  ) {materialFunction=&HeightLerp}
  n20 = OneMinus(None=@n19.Alpha)
  n21 = MaterialFunctionCall!(Material=@n9.Blended Material, Tint=@n20, TintMask=Constant() {r=1}) {materialFunction=/Engine/Functions/MaterialLayerFunctions/MatLayerBlend_Tint}
  n22 = Multiply(A=@n15.R, B=ScalarParameter() {defaultValue=1.5, parameterName=RoadMud_Trans_1})
  n23 = ScalarParameter() {defaultValue=1, parameterName=Road_Blend_Contrast}
  n24 = StaticSwitchParameter(
   True=@n6
   False=MaterialFunctionCall(
    Base Material=StaticSwitchParameter(True=@n14.Blended Material, False=@n6) {defaultValue=true, parameterName=UseMud}
    Top Material=MaterialFunctionCall(Base Material=@n21.Blended Material, Top Material=@n18.Result, Alpha=@n19.Alpha) {materialFunction=&MatLayerBlend_Standard}.Blended Material
    Alpha=MaterialFunctionCall(Transition Phase=@n22, Height Texture=@n13, Contrast=@n23) {desc=Road, materialFunction=&HeightLerp}.Alpha
   ) {materialFunction=&MatLayerBlend_Standard}.Blended Material
  ) {defaultValue=true, parameterName=Use_Cliffs}
  n25 = MaterialFunctionCall!(In=@n24) {materialFunction=/FigmentEnvironment/FN_6_20/Athena/Environments/Landscape/MaterialFunctions/Arid/MF_Athena_DesertRoad}
  n26 = MaterialFunctionCall!(In=@n17) {materialFunction=/FigmentEnvironment/FN_6_20/Athena/Environments/Landscape/MaterialFunctions/Standard/MF_Athena_Gravel_01}
  n27 = CollectionParameter() {collection=&MPC_Landscape, parameterName=Arid_Gravel_Tint}
  n28 = MaterialFunctionCall!(Mat_In=@n26.Result, BaseColor=Multiply(A=@n26.Plain_BaseColor, B=ComponentMask(None=@n27) {b=true, g=true, r=true})) {materialFunction=/FigmentEnvironment/FN_6_20/Athena/Environments/Landscape/MaterialFunctions/MF_Athena_ReplaceBaseColor}
  n29 = MaterialFunctionCall(In=@n28.Result, Color=VectorParameter() {parameterName=Gravel_Spline_Match, defaultValue=<rgba 1 0.7689 0.6574 0>}.RGB) {materialFunction=&MF_Athena_Material_BaseMultiply}
  n30 = VertexColor()
  n31 = ScalarParameter() {defaultValue=2, parameterName=Gravel_Blend_Contrast}
  n32 = OneMinus(None=@n12.Result)
  n33 = MaterialFunctionCall(
   Base Material=MaterialFunctionCall(
    Base Material=StaticSwitchParameter(True=@n25.Result, False=@n24) {parameterName=IsDesertRoad}
    Top Material=StaticSwitchParameter(True=@n29.Result, False=@n28.Result) {desc=For river spline, parameterName=MatchGravel}
    Alpha=MaterialFunctionCall(Transition Phase=@n30.B, Height Texture=@n26.Height, Contrast=@n31) {desc=Gravel, materialFunction=&HeightLerp}.Alpha
   ) {materialFunction=&MatLayerBlend_Standard}.Blended Material
   Top Material=StaticSwitchParameter(True=@n5.Result, False=StaticSwitchParameter(True=@n5.Result, False=@n2.Result) {desc=Flipped, parameterName=UseGrassAsBase}) {parameterName=NoGrass}
   Alpha=MaterialFunctionCall(
    Transition Phase=VertexColor().G
    Height Texture=StaticSwitchParameter(True=@n32, False=@n12.Result) {parameterName=UseGrassAsBase}
    Contrast=ScalarParameter() {parameterName=Grass_Blend_Contrast}
   ) {materialFunction=&HeightLerp}.Alpha
  ) {materialFunction=&MatLayerBlend_Standard}
  n34 = MaterialFunctionCall(
   Transition Phase=OneMinus(None=VertexColor().A)
   Height Texture=Constant() {r=0.5}
   Contrast=ScalarParameter() {defaultValue=0.5, parameterName=Contrasttest}
  ) {materialFunction=&HeightLerp}
  n35 = BreakMaterialAttributes!(Attr=StaticSwitchParameter(
    True=MaterialFunctionCall(Base Material=@n1.Result, Top Material=@n33.Blended Material, Alpha=@n34.Alpha) {materialFunction=&MatLayerBlend_Standard}.Blended Material
    False=@n33.Blended Material
   ) {defaultValue=true, parameterName=Use_Cliffs})
  n36 = MakeMaterialAttributes!(
   BaseColor=@n35.BaseColor
   Metallic
   Specular=Multiply(A=@n35.BaseColor, B=ScalarParameter() {defaultValue=0.5, parameterName=Spec_Amount})
   Roughness=@n35.BaseColor
   Anisotropy
   EmissiveColor
   Opacity=@n35.BaseColor
   OpacityMask
   Normal=@n35.BaseColor
   Tangent
   WorldPositionOffset
   SubsurfaceColor
   ClearCoat
   ClearCoatRoughness
   AmbientOcclusion=@n35.BaseColor
   Refraction
   CustomizedUVs_0
   CustomizedUVs_1
   CustomizedUVs_2
   CustomizedUVs_3
   CustomizedUVs_4
   CustomizedUVs_5
   CustomizedUVs_6
   CustomizedUVs_7
   PixelDepthOffset
   ShadingModel
   Displacement
  )
  n37 = TextureObject() {samplerType=SAMPLERTYPE_LinearColor, texture=/FigmentEnvironment/FN_6_20/Athena/Environments/Landscape/Textures/T_Athena_Terrain_Topo_Mask}
  n38 = ComponentMask(None=WorldPosition().XYZ) {g=true, r=true}
  n39 = ScalarParameter() {defaultValue=306200, group=TerrainColorMatch, parameterName=TerrainSize}
  n40 = Divide(A=@n39)
  n41 = AppendVector(A=@n40, B=@n40)
  n42 = VertexInterpolator(VS=Divide(A=Subtract(A=@n38, B=@n41), B=@n39))
  n43 = TextureSample(UVs=@n42.PS) {
   samplerSource=SSM_Wrap_WorldGroupSettings
   samplerType=SAMPLERTYPE_LinearColor
   texture=/FigmentEnvironment/FN_6_20/Athena/Environments/Landscape/Textures/T_Athena_Terrain_Biome_M
  }
  n44 = MaterialFunctionCall!(
   Mask 2=@n37
   UV Coordinates=@n42.PS
   RockMask
   GrassMask=Constant() {r=1}
   Arid=@n43.B
   BaseTexture=MaterialFunctionCall!(Material=@n36) {materialFunction=/Engine/Functions/MaterialLayerFunctions/MatLayerBlend_BreakBaseColor}.BaseColor
  ) {materialFunction=/FigmentEnvironment/FN_6_20/Athena/Environments/Landscape/MaterialFunctions/MF_TerrainTopoAdjustment}
  n45 = Multiply(
   A=Desaturation(
    None=@n44.Result
    Fraction=CollectionParameter() {collection=&MPC_Landscape, parameterName=Arid_EdgeDesaturation}
   )
   B=ComponentMask(None=CollectionParameter() {collection=&MPC_Landscape, parameterName=Arid_SandSediment_Tint}) {b=true, g=true, r=true}
  )
  n46 = StaticSwitchParameter(
   True=MaterialFunctionCall(Mat_In=@n36, BaseColor=LinearInterpolate(A=@n44.Result, B=@n45, Alpha=@n44.Result)) {materialFunction=&MF_Athena_ReplaceBaseColor}.Result
   False=@n36
  ) {parameterName=UseTopo}
  n47 = MaterialFunctionCall!(Material=@n46) {materialFunction=/Engine/Functions/MaterialLayerFunctions/MatLayerBlend_BreakNormal}
  n48 = MaterialFunctionCall(Float3=@n47.Normal) {materialFunction=&BreakOutFloat3Components}
  n49 = Multiply(A=Transform(None=Constant3Vector() {constant=<rgba 0 0 1 0>}), B=Constant() {r=1})
  n50 = ComponentMask(None=@n49) {g=true}
  n51 = AppendVector(A=AppendVector(A=Multiply(A=@n48.R) {constB=-1}, B=Multiply(A=@n48.B, B=@n50)), B=Multiply(A=@n48.G) {constB=-1})
  n52 = ComponentMask(None=@n49) {r=true}
  n53 = AppendVector(A=AppendVector(A=Multiply(A=@n52, B=@n48.B), B=Multiply(A=@n48.R) {constB=-1}), B=Multiply(A=@n48.G) {constB=-1})
  n54 = Constant()
  n55 = MaterialFunctionCall(In=Abs(None=@n52), Contrast=@n54) {materialFunction=&CheapContrast}
  n56 = ComponentMask(None=Multiply(A=@n47.Normal, B=Constant3Vector() {constant=<rgba 1 1 1 0>})) {g=true, r=true}
  n57 = ComponentMask(None=@n49) {b=true}
  n58 = Multiply(A=@n48.B, B=@n57)
  n59 = Abs(None=@n57)
  n60 = Normalize(VectorInput=LinearInterpolate(
    A=LinearInterpolate(A=@n51, B=@n53, Alpha=@n55.Result)
    B=AppendVector(A=@n56, B=@n58)
    Alpha=MaterialFunctionCall(In=@n59, Contrast=@n54) {materialFunction=&CheapContrast}.Result
   ))
  n61 = MaterialFunctionCall!(
   Material=@n46
   Normal=MaterialFunctionCall(
    Normal=Transform(None=@n60) {transformSourceType=TRANSFORMSOURCE_World, transformType=TRANSFORM_Tangent}
    Flatness=ScalarParameter() {group=MasterControl, parameterName=NormalFlatten}
   ) {materialFunction=&FlattenNormal}.Result
  ) {materialFunction=/Engine/Functions/MaterialLayerFunctions/MatLayerBlend_ReplaceNormals}
  n62 = WorldPosition() {worldPositionShaderOffset=WPT_ExcludeAllShaderOffsets}
  n63 = TextureSample(UVs=Divide(A=ComponentMask(None=@n62.XYZ) {g=true, r=true}, B=ScalarParameter() {defaultValue=25000, parameterName=MacroNormalScale})) {samplerType=SAMPLERTYPE_Normal, texture=/FigmentEnvironment/FN_6_20/ContentCreationTools/TextureCreation/Textures/T_BPCreated_MacroNormal_01}
  n64 = ScalarParameter() {defaultValue=0.625, parameterName=MacroNormalIntensity}
  n65 = Constant() {r=0.9}
  n66 = ShadingPathSwitch(Default=Constant() {r=0.25}, Mobile=@n65) {desc=Migrated from FeatureLevelSwitch}
  n67 = Multiply(
   A=LinearInterpolate(
    A=Constant3Vector() {constant=<rgba 0 0 1 0>}
    B=@n63.RGB
    Alpha=MaterialFunctionCall() {materialFunction=/FigmentEnvironment/FN_6_20/Athena/Environments/Landscape/MaterialFunctions/MF_TerrainDistanceFade}.Result
   )
   B=MaterialFunctionCall(X=@n64, Y=@n64, Z=QualitySwitch(Default=@n66, Low=@n65)) {materialFunction=&MakeFloat3}.Result
  )
  n68 = Add(
   A=Multiply(
    A=MaterialFunctionCall(Material=@n61.Result) {materialFunction=&MatLayerBlend_BreakNormal}.Normal
    B=Constant3Vector() {constant=<rgba 1 1 0.5 0>}
   )
   B=@n67
  )
  MaterialFunctionCall(
   Material=@n61.Result
   Normal=ShadingPathSwitch(Default=QualitySwitch(Default=@n68, Low=@n67, High=@n68, Medium=@n68), Deferred=@n68, Mobile=@n67) {desc=Migrated from FeatureLevelSwitch}
  ) {materialFunction=&MatLayerBlend_ReplaceNormals}
  Multiply(A=@n20, B=ScalarParameter())
  n69 = StaticSwitchParameter(True=@n61.Result, False=@n61.Result) {
   defaultValue=true
   desc=Weird legacy requirement for procedural grass to spawn.
   parameterName="Use Grass(true) or a BaseTilingTexture(False)"
  }
  n70 = SetMaterialAttributes!(
   MaterialAttributes=@n69
   World Position Offset=StaticSwitchParameter(
    True=MaterialFunctionCall() {materialFunction=/FigmentEnvironment/FN_6_20/Athena/Environments/Landscape/MaterialFunctions/MF_SplineMatHeightOffset}.Result
    False=@n69
   ) {parameterName=IsDesertRoad}
  ) {attributeSetTypes [#1]}
  n71 = BreakMaterialAttributes(Attr=@n70)
  n72 = MakeMaterialAttributes(
   BaseColor=@n71.BaseColor
   Specular=Multiply(A=@n71.BaseColor, B=ScalarParameter() {defaultValue=1, parameterName=SimpleMobileSpec})
   Roughness=Constant() {r=1}
   Normal=Constant3Vector() {constant=<rgba 0 0 1 0>}
   WorldPositionOffset=@n71.BaseColor
  )
  n73 = QualitySwitch(Default=@n70, Low=@n72, High=@n70, Medium=@n70)
  ShadingPathSwitch(Default=@n73, Deferred=@n73, Mobile=@n72) {desc=Migrated from FeatureLevelSwitch}
  n74 = QualitySwitch(Default=@n36, Low=@n36, High=@n46, Medium=@n46)
  ShadingPathSwitch(Default=@n74, Deferred=@n74, Mobile=@n36) {desc=Migrated from FeatureLevelSwitch}
  StaticBool()
  StaticSwitchParameter() {parameterName=IsDesertRoad_1}
  VertexNormalWS()
 ]
}
